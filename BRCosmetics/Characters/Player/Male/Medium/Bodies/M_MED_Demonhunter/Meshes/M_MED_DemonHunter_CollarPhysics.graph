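# Physics asset for the collar of the medium male Demonhunter player body, setting up physical motion for its collar bones.
# Seven collar pieces (back, plus left/right front, mid and side) are boxes, each constrained to a small base-bone box with tight swing and twist limits.
# Capsules on both neck bones and upper arms represent the surrounding body.
# The neck joint rotates freely, while the upper arms have no constraints.
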PhysicsAsset /BRCosmetics/Characters/Player/Male/Medium/Bodies/M_MED_Demonhunter/Meshes/M_MED_DemonHunter_CollarPhysics
bodies {
 dyn_collarBack [{shapeType=Box, extentX=1.4147, extentY=7.4495, extentZ=12.1136, center=<3.0604 -0.578 -0.0151>, rotation=<rot 0.0022 -89.9999 0.0292>}]
 dyn_collarBack_Base [{shapeType=Box, extentX=1.01, extentY=1.01, extentZ=1.01, rotation=<rot 0 -90 0>}]
 dyn_collarFrontBase_l [{shapeType=Box, extentX=1.01, extentY=1.01, extentZ=1.01, rotation=<rot 0 -90 0>}]
 dyn_collarFrontBase_r [{shapeType=Box, extentX=1.01, extentY=1.01, extentZ=1.01, rotation=<rot 0 -90 0>}]
 dyn_collarFront_l [{shapeType=Box, extentX=1.6025, extentY=9.6621, extentZ=8.5948, center=<-3.8606 0.2438 0.6134>, rotation=<rot 0 90.6253 163.4758>}]
 dyn_collarFront_r [{shapeType=Box, extentX=1.6025, extentY=9.6621, extentZ=8.5948, center=<3.8606 -0.2438 -0.6134>, rotation=<rot 0 90.6254 -16.5241>}]
 dyn_collarMidBase_l [{shapeType=Box, extentX=1.01, extentY=1.01, extentZ=1.01, rotation=<rot 0 -90 0>}]
 dyn_collarMidBase_r [{shapeType=Box, extentX=1.01, extentY=1.01, extentZ=1.01, rotation=<rot 0 -90 0>}]
 dyn_collarMid_l [{shapeType=Box, extentX=1.2737, extentY=6.4296, extentZ=7.0559, center=<-2.72 -0.3256 0.1091>, rotation=<rot 1.0827 -84.7608 22.8186>}]
 dyn_collarMid_r [{shapeType=Box, extentX=1.988, extentY=6.6761, extentZ=7.0559, center=<2.4552 -1.3242 -0.1405>, rotation=<rot 11.9512 -116.7964 23.3178>}]
 dyn_collarSide_Base_l [{shapeType=Box, extentX=1.01, extentY=1.01, extentZ=1.01, rotation=<rot 0 -90 0>}]
 dyn_collarSide_Base_r [{shapeType=Box, extentX=1.01, extentY=1.01, extentZ=1.01, rotation=<rot 0 -90 0>}]
 dyn_collarSide_l [{shapeType=Box, extentX=1.1678, extentY=7.1117, extentZ=7.5085, center=<-2.8842 -0.4311 -0.2631>, rotation=<rot 2.249 -82.24 9.7524>}]
 dyn_collarSide_r [{shapeType=Box, extentX=1.11, extentY=7.9595, extentZ=7.5085, center=<2.5718 -0.4892 0.3362>, rotation=<rot 1.8624 -90 9.8328>}]
 neck_01 [{shapeType=Capsule, length=5.4986, radius=9.9654, center=<9.9811 0.9725 -0.0002>, rotation=<rot 90 26.565 63.0848>}]
 neck_02 [{shapeType=Capsule, length=2.7596, radius=7.4949, center=<5.5792 -2.4418 0.0001>, rotation=<rot 0 -118.1786 0>}]
 upperarm_l [{shapeType=Capsule, length=12.4205, radius=5.6517, center=<7.1561 -2.6309 -2.47>, rotation=<rot 90 -0.0099 -0.0099>}]
 upperarm_r [{shapeType=Capsule, length=12.4205, radius=5.6517, center=<-8.404 2.2837 2.6381>, rotation=<rot 90 0.0049 0.0049>}]
}
constraints [
 {bone1Name=dyn_collarBack, bone2Name=dyn_collarBack_Base, swing1LimitDegrees=20, swing2LimitDegrees=10, twistLimitDegrees=8}
 {bone1Name=dyn_collarSide_l, bone2Name=dyn_collarSide_Base_l, swing1LimitDegrees=20, swing2LimitDegrees=10, twistLimitDegrees=2}
 {bone1Name=dyn_collarSide_r, bone2Name=dyn_collarSide_Base_r, swing1LimitDegrees=20, swing2LimitDegrees=10, twistLimitDegrees=2}
 {bone1Name=dyn_collarFront_r, bone2Name=dyn_collarFrontBase_r, swing1LimitDegrees=20, swing2LimitDegrees=10, twistLimitDegrees=8}
 {bone1Name=dyn_collarMid_r, bone2Name=dyn_collarMidBase_r, swing1LimitDegrees=20, swing2LimitDegrees=10, twistLimitDegrees=2}
 {bone1Name=dyn_collarMid_l, bone2Name=dyn_collarMidBase_l, swing1LimitDegrees=20, swing2LimitDegrees=10, twistLimitDegrees=2}
 {
  bone1Name=neck_02
  bone2Name=neck_01
  swing1LimitDegrees=45
  swing1Motion=Free
  swing2LimitDegrees=45
  swing2Motion=Free
  twistLimitDegrees=45
  twistMotion=Free
 }
 {bone1Name=dyn_collarFront_l, bone2Name=dyn_collarFrontBase_l, swing1LimitDegrees=20, swing2LimitDegrees=10, twistLimitDegrees=8}
]
