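# Control Rig for the male medium Demonhunter body that corrects the hip bag bone, dyn_bag_Base, as the left thigh moves.
# A spherical pose reader on thigh_l, with debug drawing enabled, measures how far the leg enters a tuned region.
# Its output weights a Modify Transforms step that applies a fixed translation and Euler-derived rotation to the bag.
# An identity-valued, full-weight Modify Transforms on the same bone runs first.
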ControlRigBlueprint /BRCosmetics/Characters/Player/Male/Medium/Bodies/M_MED_Demonhunter/Meshes/M_MED_Demon_Hunter_CtrlRig
graphs [
 RigVMFunctionLibrary {links [], nodes []}
 RigVMModel {nodes [
   RigUnit_BeginExecution = &RigUnit_BeginExecution()
   &RigUnit_ModifyTransforms(
    ItemToModify=((Item=(Type=Bone,Name="dyn_bag_Base"),Transform=(Rotation=(X=0,Y=0,Z=0,W=1),Translation=(X=0,Y=0,Z=0),Scale3D=(X=1,Y=1,Z=1))))
    Weight=1
    ExecutePin=@RigUnit_BeginExecution.ExecutePin
   )
   > n1 = &RigUnit_SphericalPoseReader(
    ActiveRegionScaleFactors=(PositiveWidth=0.5,NegativeWidth=0.5,PositiveHeight=0.5,NegativeHeight=0.5)
    ActiveRegionSize=0.1
    Debug=(bDrawDebug=True,DebugScale=25,DebugSegments=20,DebugThickness=0.25)
    DriverAxis=(X=-1,Y=0,Z=0)
    DriverItem=(Type=Bone,Name="thigh_l")
    FalloffRegionScaleFactors=(PositiveWidth=1.3,NegativeWidth=1.3,PositiveHeight=1.3,NegativeHeight=1.3)
    FalloffSize=0.2
    OptionalParentItem=(Type=Bone)
    RotationOffset=(X=150,Y=0,Z=0)
   )
   > &RigUnit_ModifyTransforms(
    ItemToModify=((Item=(Type=Bone,Name="dyn_bag_Base"),Transform=(Rotation=(X=0,Y=0,Z=0,W=1),Translation=(X=3,Y=-2,Z=3),Scale3D=(X=1,Y=1,Z=1))))
    ItemToModify.0.Transform.Rotation=&RigVMFunction_MathQuaternionFromEuler(Euler=(X=10,Y=-30,Z=5)).Result
    Weight=@n1.OutputParam
   )
  ]}
]
hierarchy [
 root {loc=<0 0 0>, rot=<0 0 0 1>}
 pelvis {parent=root, loc=<0 -0.7071 93.1191>, rot=<0.0127 0.707 -0.0127 -0.707>}
 spine_01 {parent=pelvis, loc=<6.7494 -0.6388 0>, rot=<0 0 0.034 -0.9994>}
 spine_02 {parent=spine_01, loc=<7.2518 -0.035 0>, rot=<0 0 0.0238 -0.9997>}
 spine_03 {parent=spine_02, loc=<6.0104 0 0>, rot=<0 0 0.0149 -0.9999>}
 spine_04 {parent=spine_03, loc=<8.1024 0 0>, rot=<0 0 -0.0132 -0.9999>}
 spine_05 {parent=spine_04, loc=<10.0146 0 0>, rot=<0 0 -0.0015 -1>}
 clavicle_l {parent=spine_05, loc=<4.3575 -0.1466 -2.3223>, rot=<0.0465 -0.7289 -0.009 -0.6829>}
 upperarm_l {parent=clavicle_l, loc=<15.5111 0 0>, rot=<0.0226 -0.3234 -0.0088 -0.9459>}
 lowerarm_l {parent=upperarm_l, loc=<26.6514 0 0>, rot=<0.0391 -0.0099 -0.2453 -0.9686>}
 hand_l {parent=lowerarm_l, loc=<22.3491 0 0>, rot=<-0.7081 -0.0507 0.0975 -0.6975>}
 index_metacarpal_l {parent=hand_l, loc=<1.6535 0.0183 -2.96>, rot=<0.0158 -0.0124 -0.0173 -0.9996>}
 index_01_l {parent=index_metacarpal_l, loc=<8.3008 0.0357 -0.0524>, rot=<-0.0044 0.0485 0.0924 -0.9945>}
 index_02_l {parent=index_01_l, loc=<3.3587 -0.3004 0.0452>, rot=<-0.002 0.0027 0.3225 -0.9466>}
 index_03_l {parent=index_02_l, loc=<2.5578 -0.123 -0.0252>, rot=<-0.002 0.0067 0.2558 -0.9667>}
 middle_metacarpal_l {parent=hand_l, loc=<2.1856 -0.1725 -0.5984>, rot=<0.006 0.0168 -0.0329 -0.9993>}
 middle_01_l {parent=middle_metacarpal_l, loc=<7.3275 -0.2447 0.0129>, rot=<-0.0208 0.0042 0.0858 -0.9961>}
 middle_02_l {parent=middle_01_l, loc=<4.0123 0.1401 0.1175>, rot=<-0.0218 0.0258 0.3512 -0.9357>}
 middle_03_l {parent=middle_02_l, loc=<2.6077 0.0534 0.1373>, rot=<0.0082 0.0188 0.1708 -0.9851>}
 pinky_metacarpal_l {parent=hand_l, loc=<1.7616 -0.9753 3.3814>, rot=<-0.0397 0.0814 -0.069 -0.9935>}
 pinky_01_l {parent=pinky_metacarpal_l, loc=<7.1709 -0.4831 -0.1094>, rot=<-0.0654 -0.0035 0.2242 -0.9723>}
 pinky_02_l {parent=pinky_01_l, loc=<2.8274 0.0893 -0.1043>, rot=<0.0037 0.0011 0.2854 -0.9584>}
 pinky_03_l {parent=pinky_02_l, loc=<2.2046 0.0307 -0.0078>, rot=<-0.0002 0.0098 0.1445 -0.9895>}
 ring_metacarpal_l {parent=hand_l, loc=<2.2074 -0.3625 1.3638>, rot=<-0.0284 0.0722 -0.0458 -0.9959>}
 ring_01_l {parent=ring_metacarpal_l, loc=<7.0295 -0.6197 -0.1446>, rot=<-0.0489 -0.0252 0.1892 -0.9804>}
 ring_02_l {parent=ring_01_l, loc=<3.6174 0.4647 -0.1027>, rot=<0 0 0.2806 -0.9598>}
 ring_03_l {parent=ring_02_l, loc=<2.4152 0.114 0.0199>, rot=<0 0 0.1189 -0.9929>}
 thumb_01_l {parent=hand_l, loc=<1.3118 -0.3191 -2.459>, rot=<0.6546 -0.4042 0.0664 -0.6354>}
 thumb_02_l {parent=thumb_01_l, loc=<5.9336 -0.0448 0.0401>, rot=<-0.0037 0.0213 0.298 -0.9543>}
 thumb_03_l {parent=thumb_02_l, loc=<2.2223 0.3682 -0.0907>, rot=<0.0118 -0.013 0.0066 -0.9998>}
 lowerarm_twist_01_l {parent=lowerarm_l, loc=<17.8549 0 0>, rot=<0 0 0 1>}
 lowerarm_twist_02_l {parent=lowerarm_l, loc=<9.6199 0 0>, rot=<0 0 0 1>}
 upperarm_twist_01_l {parent=upperarm_l, loc=<0.0041 0 0>, rot=<0 0 0 1>}
 upperarm_twist_02_l {parent=upperarm_l, loc=<13.7941 0 0>, rot=<0 0 0 1>}
 clavicle_r {parent=spine_05, loc=<4.3575 -0.1466 2.3223>, rot=<0.7289 0.0465 -0.6829 0.009>}
 upperarm_r {parent=clavicle_r, loc=<-15.5111 0 0>, rot=<0.0226 -0.3234 -0.0088 -0.9459>}
 lowerarm_r {parent=upperarm_r, loc=<-26.6515 0 0>, rot=<0.0391 -0.0099 -0.2453 -0.9686>}
 hand_r {parent=lowerarm_r, loc=<-22.3492 0 0>, rot=<-0.7081 -0.0507 0.0975 -0.6975>}
 index_metacarpal_r {parent=hand_r, loc=<-1.6534 -0.0181 2.9601>, rot=<0.0158 -0.0124 -0.0173 -0.9996>}
 index_01_r {parent=index_metacarpal_r, loc=<-8.3008 -0.0357 0.0524>, rot=<-0.0044 0.0485 0.0924 -0.9945>}
 index_02_r {parent=index_01_r, loc=<-3.3589 0.3003 -0.0453>, rot=<-0.002 0.0027 0.3225 -0.9466>}
 index_03_r {parent=index_02_r, loc=<-2.5578 0.1231 0.0251>, rot=<-0.002 0.0067 0.2558 -0.9667>}
 middle_metacarpal_r {parent=hand_r, loc=<-2.1856 0.1726 0.5984>, rot=<0.006 0.0168 -0.0329 -0.9993>}
 middle_01_r {parent=middle_metacarpal_r, loc=<-7.3273 0.2447 -0.0129>, rot=<-0.0208 0.0042 0.0858 -0.9961>}
 middle_02_r {parent=middle_01_r, loc=<-4.0142 -0.1411 -0.07>, rot=<-0.0217 0.0259 0.3512 -0.9357>}
 middle_03_r {parent=middle_02_r, loc=<-2.6074 -0.0534 -0.1312>, rot=<0.0083 0.0188 0.1708 -0.9851>}
 pinky_metacarpal_r {parent=hand_r, loc=<-1.7617 0.9753 -3.3814>, rot=<-0.0397 0.0814 -0.069 -0.9935>}
 pinky_01_r {parent=pinky_metacarpal_r, loc=<-7.1707 0.4832 0.1094>, rot=<-0.0654 -0.0035 0.2242 -0.9723>}
 pinky_02_r {parent=pinky_01_r, loc=<-2.8274 -0.0892 0.1043>, rot=<0.0037 0.0011 0.2854 -0.9584>}
 pinky_03_r {parent=pinky_02_r, loc=<-2.2045 -0.031 0.0078>, rot=<-0.0002 0.0098 0.1445 -0.9895>}
 ring_metacarpal_r {parent=hand_r, loc=<-2.2073 0.3627 -1.3638>, rot=<-0.0284 0.0722 -0.0458 -0.9959>}
 ring_01_r {parent=ring_metacarpal_r, loc=<-7.0297 0.6197 0.1446>, rot=<-0.0489 -0.0252 0.1892 -0.9804>}
 ring_02_r {parent=ring_01_r, loc=<-3.6171 -0.4648 0.1027>, rot=<0 0 0.2806 -0.9598>}
 ring_03_r {parent=ring_02_r, loc=<-2.4153 -0.1141 -0.0199>, rot=<0 0 0.1189 -0.9929>}
 thumb_01_r {parent=hand_r, loc=<-1.3119 0.3192 2.459>, rot=<0.6546 -0.4042 0.0664 -0.6354>}
 thumb_02_r {parent=thumb_01_r, loc=<-5.9335 0.0447 -0.0401>, rot=<-0.0037 0.0213 0.298 -0.9543>}
 thumb_03_r {parent=thumb_02_r, loc=<-2.2223 -0.3682 0.0907>, rot=<0.0118 -0.013 0.0066 -0.9998>}
 lowerarm_twist_01_r {parent=lowerarm_r, loc=<-17.855 0 0>, rot=<0 0 0 1>}
 lowerarm_twist_02_r {parent=lowerarm_r, loc=<-9.6199 0 0>, rot=<0 0 0 1>}
 upperarm_twist_01_r {parent=upperarm_r, loc=<-0.0041 0 0>, rot=<0 0 0 1>}
 upperarm_twist_02_r {parent=upperarm_r, loc=<-13.7942 0 0>, rot=<0 0 0 1>}
 neck_01 {parent=spine_05, loc=<11.3138 0.1529 0>, rot=<0 0 -0.2003 -0.9797>}
 neck_02 {parent=neck_01, loc=<4.5627 0 0>, rot=<0 0 0.0504 -0.9987>}
 head {parent=neck_02, loc=<5.3914 0 0>, rot=<0 0 0.115 -0.9934>}
 dyn_collarMidBaseDriven_l {parent=spine_05, loc=<12.0502 0.3128 -9.8572>, rot=<-0.1376 -0.8653 -0.439 0.199>}
 dyn_collarMidBase_l {parent=dyn_collarMidBaseDriven_l, loc=<0 0 0>, rot=<0 0 0 1>}
 dyn_collarMid_l {parent=dyn_collarMidBase_l, loc=<0 0 0>, rot=<0 0 0 1>}
 dyn_collarMidBaseDriven_r {parent=spine_05, loc=<12.0502 0.3128 9.8572>, rot=<0.8653 -0.1376 0.199 0.439>}
 dyn_collarMidBase_r {parent=dyn_collarMidBaseDriven_r, loc=<0 0 0>, rot=<-0.0842 0.0226 -0.2578 -0.9623>}
 dyn_collarMid_r {parent=dyn_collarMidBase_r, loc=<0 0 0>, rot=<0 0 0 1>}
 dyn_collarFrontBaseDriven_r {parent=spine_05, loc=<8.5822 7.1682 11.4346>, rot=<-0.7653 0.0566 -0.3833 -0.5141>}
 dyn_collarFrontBase_r {parent=dyn_collarFrontBaseDriven_r, loc=<0 0 0>, rot=<-0.0283 0.2146 -0.1274 -0.9679>}
 dyn_collarFront_r {parent=dyn_collarFrontBase_r, loc=<0 0 0>, rot=<0 0 0 1>}
 dyn_collarFrontBaseDriven_l {parent=spine_05, loc=<8.5822 7.1682 -11.4346>, rot=<-0.0566 -0.7653 -0.5141 0.3833>}
 dyn_collarFrontBase_l {parent=dyn_collarFrontBaseDriven_l, loc=<0 0 0>, rot=<0 0 0 1>}
 dyn_collarFront_l {parent=dyn_collarFrontBase_l, loc=<0 0 0>, rot=<0 0 0 1>}
 dyn_hoodiePull_01_Base_l {parent=spine_05, loc=<9.2438 11.9333 -4.0781>, rot=<0.9902 -0.0498 0.1291 -0.0169>}
 dyn_hoodiePull_01_l {parent=dyn_hoodiePull_01_Base_l, loc=<0 0 0>, rot=<0 0 0 1>}
 dyn_hoodiePull_02_l {parent=dyn_hoodiePull_01_l, loc=<-3.9059 -0.0001 0.0001>, rot=<-0.006 -0.041 0.0349 -0.9985>}
 dyn_hoodiePull_03_l {parent=dyn_hoodiePull_02_l, loc=<-2.2392 0 0.0001>, rot=<-0.0009 -0.0071 -0.014 -0.9999>}
 dyn_hoodiePull_01_Base_r {parent=spine_05, loc=<9.1571 12.2264 3.6292>, rot=<-0.0498 -0.9902 0.0169 0.1291>}
 dyn_hoodiePull_01_r {parent=dyn_hoodiePull_01_Base_r, loc=<0 0 0>, rot=<0 0 0 1>}
 dyn_hoodiePull_02_r {parent=dyn_hoodiePull_01_r, loc=<3.9063 0 0>, rot=<-0.006 -0.041 0.0349 -0.9985>}
 dyn_hoodiePull_03_r {parent=dyn_hoodiePull_02_r, loc=<2.2395 0 0>, rot=<-0.0009 -0.0071 -0.014 -0.9999>}
 dyn_collarSide_Base_r {parent=spine_05, loc=<13.8294 -3.6831 6.9717>, rot=<0.9457 -0.1094 0.1401 0.2723>}
 dyn_collarSide_r {parent=dyn_collarSide_Base_r, loc=<0 0 0>, rot=<0 0 -0.0872 -0.9962>}
 dyn_collarSide_Base_l {parent=spine_05, loc=<13.8294 -3.6831 -6.9717>, rot=<-0.1094 -0.9457 -0.2723 0.1401>}
 dyn_collarSide_l {parent=dyn_collarSide_Base_l, loc=<0 0 0>, rot=<0 0 0 1>}
 dyn_collarBack_Base {parent=spine_05, loc=<14.2887 -5.5452 -0.0164>, rot=<0.9781 -0.2079 0 0>}
 dyn_collarBack {parent=dyn_collarBack_Base, loc=<0 0 0>, rot=<0 0 0 1>}
 thigh_split_l {parent=pelvis, loc=<-2.5693 -0.0109 -7.7501>, rot=<0.0012 0.0572 0.0155 -0.9982>}
 dyn_bag_Base {parent=thigh_split_l, loc=<0.281 -0.3508 -4.7621>, rot=<0.3336 -0.6424 -0.6233 -0.2957>}
 dyn_bag_01 {parent=dyn_bag_Base, loc=<0 0 0>, rot=<0 0 0 1>}
 dyn_bag_02 {parent=dyn_bag_01, loc=<5.8636 1.0359 0.5189>, rot=<0.0358 -0.0207 -0.3576 -0.933>}
 thigh_l {parent=pelvis, loc=<-2.5693 -0.0109 -7.7501>, rot=<0.0012 0.0572 0.0155 -0.9982>}
 calf_l {parent=thigh_l, loc=<-39.0484 0 0>, rot=<0.0006 -0.0161 -0.0391 -0.9991>}
 calf_twist_02_l {parent=calf_l, loc=<-18.8064 -0.0967 0.0397>, rot=<0.0237 0.0095 -0.0035 -0.9997>}
 foot_l {parent=calf_l, loc=<-42.561 0.1077 0.4292>, rot=<0.0418 -0.0455 0.0343 -0.9975>}
 ball_l {parent=foot_l, loc=<-6.1599 13.8264 0.6806>, rot=<0.0001 -0.0001 -0.7063 -0.7079>}
 thigh_twist_01_l {parent=thigh_l, loc=<-21.6303 0 0>, rot=<0.0013 0 0 -1>}
 dyn_kneeFront_Base_l {parent=thigh_l, loc=<-39.6567 7.1196 -2.1147>, rot=<0.9793 0.1109 -0.0137 0.1689>}
 dyn_kneeFront_l {parent=dyn_kneeFront_Base_l, loc=<0 0 0>, rot=<0 0 0 1>}
 dyn_kneeOut_Base_l {parent=thigh_l, loc=<-37.7456 -0.7274 -8.2496>, rot=<-0.7059 -0.0521 -0.0423 -0.7051>}
 dyn_kneeOut_l {parent=dyn_kneeOut_Base_l, loc=<0 0 0>, rot=<0 0 0 1>}
 dyn_kneeBack_Base_l {parent=thigh_l, loc=<-36.5008 -8.7661 -0.515>, rot=<-0.0439 -0.0532 0.0046 -0.9976>}
 dyn_kneeBack_l {parent=dyn_kneeBack_Base_l, loc=<0 0 0>, rot=<0 0 0 1>}
 dyn_kneeIn_Base_l {parent=thigh_l, loc=<-39.1455 1.0592 7.3812>, rot=<0.7918 0.0642 -0.0334 -0.6065>}
 dyn_kneeIn_l {parent=dyn_kneeIn_Base_l, loc=<0 0 0>, rot=<0 0 0 1>}
 thigh_r {parent=pelvis, loc=<-2.5691 -0.0109 7.7501>, rot=<0.0572 -0.0012 0.9982 0.0155>}
 calf_r {parent=thigh_r, loc=<39.0487 0 0>, rot=<0.0006 -0.0161 -0.0391 -0.9991>}
 calf_twist_02_r {parent=calf_r, loc=<18.8064 0.0967 -0.0397>, rot=<0.0237 0.0095 -0.0035 -0.9997>}
 foot_r {parent=calf_r, loc=<42.5612 -0.1077 -0.4292>, rot=<0.0418 -0.0455 0.0343 -0.9975>}
 ball_r {parent=foot_r, loc=<6.1598 -13.8264 -0.6806>, rot=<0.0001 -0.0001 -0.7063 -0.7079>}
 thigh_twist_01_r {parent=thigh_r, loc=<21.6303 0 0>, rot=<0.0013 0 0 -1>}
 dyn_kneeOut_Base_r {parent=thigh_r, loc=<37.7457 0.7274 8.2495>, rot=<-0.7059 -0.0521 -0.0423 -0.7051>}
 dyn_kneeOut_r {parent=dyn_kneeOut_Base_r, loc=<0 0 0>, rot=<0 0 0 1>}
 dyn_kneeIn_Base_r {parent=thigh_r, loc=<39.1457 -1.0592 -7.3812>, rot=<0.7918 0.0642 -0.0334 -0.6065>}
 dyn_kneeIn_r {parent=dyn_kneeIn_Base_r, loc=<0 0 0>, rot=<0 0 0 1>}
 dyn_kneeBack_Base_r {parent=thigh_r, loc=<36.501 8.7661 0.515>, rot=<-0.0439 -0.0532 0.0046 -0.9976>}
 dyn_kneeBack_r {parent=dyn_kneeBack_Base_r, loc=<0 0 0>, rot=<0 0 0 1>}
 dyn_kneeFront_Base_r {parent=thigh_r, loc=<39.6569 -7.1196 2.1146>, rot=<0.9793 0.1109 -0.0137 0.1689>}
 dyn_kneeFront_r {parent=dyn_kneeFront_Base_r, loc=<0 0 0>, rot=<0 0 0 1>}
 dyn_knife {parent=thigh_r, loc=<10.7271 1.293 8.8808>, rot=<0.7577 0.1324 -0.1677 -0.6167>}
 dyn_jacketFront_Base_l {parent=pelvis, loc=<20.9786 4.7366 -13.882>, rot=<0.4961 0.0066 -0.0714 -0.8653>}
 dyn_jacketFront_l {parent=dyn_jacketFront_Base_l, loc=<0 0 0>, rot=<0 0 0 1>}
 dyn_jacketZipper {parent=dyn_jacketFront_l, loc=<-13.0841 1.858 9.4659>, rot=<0.9392 -0.1656 -0.0522 -0.2961>}
 dyn_jacketBack_Base {parent=pelvis, loc=<20.0263 -14.7055 0.0002>, rot=<-0.1149 -0.9934 0 0>}
 dyn_jacketBack {parent=dyn_jacketBack_Base, loc=<0 0 0>, rot=<0 0 0 1>}
 dyn_jacketSideBack_Base_r {parent=pelvis, loc=<19.9458 -12.2744 10.1352>, rot=<-0.0417 -0.9227 0.3823 0.0265>}
 dyn_jacketSideBack_r {parent=dyn_jacketSideBack_Base_r, loc=<0 0 0>, rot=<0 0 0 1>}
 dyn_jacketSideBack_Base_l {parent=pelvis, loc=<20.4395 -12.3375 -9.7118>, rot=<0.9227 -0.0417 0.0265 -0.3823>}
 dyn_jacketSideBack_l {parent=dyn_jacketSideBack_Base_l, loc=<0 0 0>, rot=<0 0 0 1>}
 dyn_jacketSide_Base_l {parent=pelvis, loc=<19.6491 -4.2462 -15.157>, rot=<0.6737 -0.0113 -0.0371 -0.738>}
 dyn_jacketSide_l {parent=dyn_jacketSide_Base_l, loc=<0 0 0>, rot=<0 0 0 1>}
 dyn_jacketSide_Base_r {parent=pelvis, loc=<19.6487 -4.2462 15.1574>, rot=<-0.0113 -0.6737 0.738 -0.0371>}
 dyn_jacketSide_r {parent=dyn_jacketSide_Base_r, loc=<0 0 0>, rot=<0 0 0 1>}
 dyn_jacketFront_Base_r {parent=pelvis, loc=<20.8597 4.754 13.8824>, rot=<0.0066 -0.4961 0.8653 -0.0714>}
 dyn_jacketFront_r {parent=dyn_jacketFront_Base_r, loc=<0 0 0>, rot=<0 0 0 1>}
]
