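# Physics asset for the male medium Demon Hunter cosmetic body, set up for secondary motion of jacket, hoodie pulls, bag and knife.
# Capsules on pelvis, spine and thighs form the body core, joined by free-angle constraints.
# Jacket panels and a twist-locked bag chain hang from pelvis via small base bodies, hoodie-pull chains likewise from spine_02.
# Angular limits restrain these parts, the zipper on the left front panel and the right-thigh knife.
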PhysicsAsset /BRCosmetics/Characters/Player/Male/Medium/Bodies/M_MED_Demonhunter/Meshes/M_MED_Demon_Hunter_Physics
bodies {
 dyn_bag_01 [{shapeType=Capsule, length=2.6812, radius=4.2412, center=<3.2635 0 -0.0002>, rotation=<rot 90 0 0>}]
 dyn_bag_02 [{shapeType=Capsule, length=2.5746, radius=4.6474, center=<2.9559 -0.3991 0.0001>, rotation=<rot 90 0 0>}]
 dyn_bag_Base [{shapeType=Capsule, length=0.505, radius=0.505, rotation=<rot 0 -90 0>}]
 dyn_hoodiePull_01_Base_l [{shapeType=Capsule, length=0.505, radius=0.505, rotation=<rot 0 -90 0>}]
 dyn_hoodiePull_01_Base_r [{shapeType=Capsule, length=0.505, radius=0.505, rotation=<rot 0 -90 0>}]
 dyn_hoodiePull_01_l [{shapeType=Capsule, length=3.8969, radius=0.5382, center=<-1.8487 0 0>, rotation=<rot 90 0 0>}]
 dyn_hoodiePull_01_r [{shapeType=Capsule, length=4.001, radius=0.5318, center=<2.0462 0.0001 0>, rotation=<rot 90 0 0>}]
 dyn_hoodiePull_02_l [{shapeType=Capsule, length=2.2131, radius=0.5235, center=<-1.1547 0 0>, rotation=<rot 90 0 0>}]
 dyn_hoodiePull_02_r [{shapeType=Capsule, length=2.3573, radius=0.5096, center=<1.1945 -0.0001 0>, rotation=<rot 90 0 0>}]
 dyn_hoodiePull_03_l [{shapeType=Capsule, length=1.7965, radius=0.5, center=<-1.0305 0 0>, rotation=<rot 90 0 0>}]
 dyn_hoodiePull_03_r [{shapeType=Capsule, length=1.604, radius=0.5, center=<1.0467 0 0>, rotation=<rot 90 0 0>}]
 dyn_jacketBack [{shapeType=Box, extentX=1.1413, extentY=10.2475, extentZ=9.9622, center=<4.459 -1.0056 0>, rotation=<rot 0 77.2768 0>}]
 dyn_jacketBack_Base [{shapeType=Box, extentX=1.01, extentY=1.01, extentZ=1.01, rotation=<rot 0 -90 0>}]
 dyn_jacketFront_Base_l [{shapeType=Box, extentX=1.01, extentY=1.01, extentZ=1.01, rotation=<rot 0 -90 0>}]
 dyn_jacketFront_Base_r [{shapeType=Box, extentX=1.01, extentY=1.01, extentZ=1.01, rotation=<rot 0 -90 0>}]
 dyn_jacketFront_l [{shapeType=Box, extentX=1.5668, extentY=13.9597, extentZ=13.4922, center=<-5.8929 0.8861 0.3904>, rotation=<rot -0.0001 80.0774 -180>}]
 dyn_jacketFront_r [{shapeType=Box, extentX=1.5668, extentY=13.9597, extentZ=13.4922, center=<5.8929 -0.8861 -0.3904>, rotation=<rot -0.0001 80.0774 0>}]
 dyn_jacketSideBack_Base_l [{shapeType=Box, extentX=1.01, extentY=1.01, extentZ=1.01, rotation=<rot 0 -90 0>}]
 dyn_jacketSideBack_Base_r [{shapeType=Box, extentX=1.01, extentY=1.01, extentZ=1.01, rotation=<rot 0 -90 0>}]
 dyn_jacketSideBack_l [{shapeType=Box, extentX=0.9915, extentY=9.9834, extentZ=10.0228, center=<-4.4757 0.4346 0.0639>, rotation=<rot -6.9532 84.0798 -174.6534>}]
 dyn_jacketSideBack_r [{shapeType=Box, extentX=0.9915, extentY=9.9834, extentZ=10.0228, center=<4.5705 0 -0.0551>, rotation=<rot 0 89.9999 5.3271>}]
 dyn_jacketSide_Base_l [{shapeType=Box, extentX=1.01, extentY=1.01, extentZ=1.01, rotation=<rot 0 -90 0>}]
 dyn_jacketSide_Base_r [{shapeType=Box, extentX=1.01, extentY=1.01, extentZ=1.01, rotation=<rot 0 -90 0>}]
 dyn_jacketSide_l [{shapeType=Box, extentX=1.1523, extentY=10.6439, extentZ=9.2017, center=<-4.7081 0.3128 -0.0773>, rotation=<rot -13.8489 86.9994 -179.2812>}]
 dyn_jacketSide_r [{shapeType=Box, extentX=1.1523, extentY=10.6439, extentZ=9.2017, center=<4.7081 -0.3128 0.0773>, rotation=<rot -13.8489 86.9994 0.7187>}]
 dyn_jacketZipper [{shapeType=Box, extentX=0.962, extentY=5.5642, extentZ=4.037, center=<-2.3572 -0.0516 0>, rotation=<rot 0 90 0>}]
 dyn_knife [{shapeType=Box, extentX=3.3117, extentY=9.4837, extentZ=36.1572, center=<5.4284 -0.7665 -0.1519>, rotation=<rot 0 95.2319 -88.024>}]
 pelvis [{shapeType=Capsule, length=8.1113, radius=11.7281, center=<0 0.0001 0.5072>, rotation=<rot 0 90 0>}]
 spine_01 [{shapeType=Capsule, length=3.985, radius=11.7496, rotation=<rot 0 -90 0>}]
 spine_02 [{shapeType=Capsule, length=8.9075, radius=13.4572, center=<-2.3718 0.7127 0>, rotation=<rot 90 0 0>}]
 spine_05 [{shapeType=Capsule, length=8.0253, radius=11.5041, center=<2.5758 1.1929 -0.1111>, rotation=<rot 0 90 0>}]
 thigh_l [{shapeType=Capsule, length=13, radius=5.5, shapeName=thigh_l_capsule, center=<-14.1809 0 0>, rotation=<rot 90 0 0>}]
 thigh_r [{shapeType=Capsule, length=9.9317, radius=7.7054, center=<18.6776 0 -0.2841>, rotation=<rot 84.5544 0 0>}]
}
constraints [
 {
  bone1Name=dyn_jacketFront_Base_l
  bone2Name=pelvis
  swing1LimitDegrees=45
  swing1Motion=Free
  swing2LimitDegrees=45
  swing2Motion=Free
  twistLimitDegrees=45
  twistMotion=Free
 }
 {bone1Name=dyn_jacketZipper, bone2Name=dyn_jacketFront_l, swing1LimitDegrees=45, swing2LimitDegrees=1, twistLimitDegrees=2}
 {
  bone1Name=dyn_jacketBack_Base
  bone2Name=pelvis
  swing1LimitDegrees=45
  swing1Motion=Free
  swing2LimitDegrees=45
  swing2Motion=Free
  twistLimitDegrees=45
  twistMotion=Free
 }
 {bone1Name=dyn_jacketBack, bone2Name=dyn_jacketBack_Base, swing1LimitDegrees=45, swing2LimitDegrees=15, twistLimitDegrees=8}
 {
  bone1Name=dyn_jacketSideBack_Base_r
  bone2Name=pelvis
  swing1LimitDegrees=45
  swing1Motion=Free
  swing2LimitDegrees=45
  swing2Motion=Free
  twistLimitDegrees=45
  twistMotion=Free
 }
 {bone1Name=dyn_jacketSideBack_r, bone2Name=dyn_jacketSideBack_Base_r, swing1LimitDegrees=45, swing2LimitDegrees=15, twistLimitDegrees=8}
 {
  bone1Name=dyn_jacketSideBack_Base_l
  bone2Name=pelvis
  swing1LimitDegrees=45
  swing1Motion=Free
  swing2LimitDegrees=45
  swing2Motion=Free
  twistLimitDegrees=45
  twistMotion=Free
 }
 {
  bone1Name=dyn_jacketSide_Base_l
  bone2Name=pelvis
  swing1LimitDegrees=45
  swing1Motion=Free
  swing2LimitDegrees=45
  swing2Motion=Free
  twistLimitDegrees=45
  twistMotion=Free
 }
 {
  bone1Name=dyn_jacketSide_Base_r
  bone2Name=pelvis
  swing1LimitDegrees=45
  swing1Motion=Free
  swing2LimitDegrees=45
  swing2Motion=Free
  twistLimitDegrees=45
  twistMotion=Free
 }
 {bone1Name=dyn_jacketSide_r, bone2Name=dyn_jacketSide_Base_r, swing1LimitDegrees=45, swing2LimitDegrees=15, twistLimitDegrees=8}
 {
  bone1Name=dyn_jacketFront_Base_r
  bone2Name=pelvis
  swing1LimitDegrees=45
  swing1Motion=Free
  swing2LimitDegrees=45
  swing2Motion=Free
  twistLimitDegrees=45
  twistMotion=Free
 }
 {bone1Name=dyn_jacketFront_r, bone2Name=dyn_jacketFront_Base_r, swing1LimitDegrees=45, swing2LimitDegrees=25, twistLimitDegrees=8}
 {bone1Name=dyn_jacketSideBack_l, bone2Name=dyn_jacketSideBack_Base_l, swing1LimitDegrees=45, swing2LimitDegrees=15, twistLimitDegrees=8}
 {bone1Name=dyn_jacketSide_l, bone2Name=dyn_jacketSide_Base_l, swing1LimitDegrees=45, swing2LimitDegrees=15, twistLimitDegrees=8}
 {bone1Name=dyn_jacketFront_l, bone2Name=dyn_jacketFront_Base_l, swing1LimitDegrees=45, swing2LimitDegrees=25, twistLimitDegrees=8}
 {
  bone1Name=spine_01
  bone2Name=pelvis
  swing1LimitDegrees=45
  swing1Motion=Free
  swing2LimitDegrees=45
  swing2Motion=Free
  twistLimitDegrees=45
  twistMotion=Free
 }
 {
  bone1Name=spine_02
  bone2Name=spine_01
  swing1LimitDegrees=45
  swing1Motion=Free
  swing2LimitDegrees=45
  swing2Motion=Free
  twistLimitDegrees=45
  twistMotion=Free
 }
 {
  bone1Name=dyn_hoodiePull_01_Base_l
  bone2Name=spine_02
  swing1LimitDegrees=45
  swing1Motion=Free
  swing2LimitDegrees=45
  swing2Motion=Free
  twistLimitDegrees=45
  twistMotion=Free
 }
 {bone1Name=dyn_hoodiePull_01_l, bone2Name=dyn_hoodiePull_01_Base_l, swing1LimitDegrees=20, swing2LimitDegrees=20, twistLimitDegrees=5}
 {bone1Name=dyn_hoodiePull_02_l, bone2Name=dyn_hoodiePull_01_l, swing1LimitDegrees=20, swing2LimitDegrees=20, twistLimitDegrees=5}
 {bone1Name=dyn_hoodiePull_03_l, bone2Name=dyn_hoodiePull_02_l, swing1LimitDegrees=20, swing2LimitDegrees=20, twistLimitDegrees=5}
 {
  bone1Name=dyn_hoodiePull_01_Base_r
  bone2Name=spine_02
  swing1LimitDegrees=45
  swing1Motion=Free
  swing2LimitDegrees=45
  swing2Motion=Free
  twistLimitDegrees=45
  twistMotion=Free
 }
 {bone1Name=dyn_hoodiePull_01_r, bone2Name=dyn_hoodiePull_01_Base_r, swing1LimitDegrees=20, swing2LimitDegrees=20, twistLimitDegrees=5}
 {bone1Name=dyn_hoodiePull_02_r, bone2Name=dyn_hoodiePull_01_r, swing1LimitDegrees=20, swing2LimitDegrees=20, twistLimitDegrees=5}
 {bone1Name=dyn_hoodiePull_03_r, bone2Name=dyn_hoodiePull_02_r, swing1LimitDegrees=20, swing2LimitDegrees=20, twistLimitDegrees=5}
 {
  bone1Name=spine_05
  bone2Name=spine_02
  swing1LimitDegrees=45
  swing1Motion=Free
  swing2LimitDegrees=45
  swing2Motion=Free
  twistLimitDegrees=45
  twistMotion=Free
 }
 {
  bone1Name=thigh_r
  bone2Name=pelvis
  swing1LimitDegrees=45
  swing1Motion=Free
  swing2LimitDegrees=45
  swing2Motion=Free
  twistLimitDegrees=45
  twistMotion=Free
 }
 {bone1Name=dyn_knife, bone2Name=thigh_r, swing1LimitDegrees=15, swing2LimitDegrees=15, twistLimitDegrees=10}
 {
  bone1Name=dyn_bag_Base
  bone2Name=pelvis
  swing1LimitDegrees=45
  swing1Motion=Free
  swing2LimitDegrees=45
  swing2Motion=Free
  twistLimitDegrees=45
  twistMotion=Free
 }
 {bone1Name=dyn_bag_01, bone2Name=dyn_bag_Base, swing1LimitDegrees=35, swing2LimitDegrees=35, twistLimitDegrees=10, twistMotion=Locked}
 {bone1Name=dyn_bag_02, bone2Name=dyn_bag_01, swing1LimitDegrees=35, swing2LimitDegrees=35, twistLimitDegrees=10, twistMotion=Locked}
 {
  bone1Name=thigh_l
  bone2Name=pelvis
  swing1LimitDegrees=45
  swing1Motion=Free
  swing2LimitDegrees=45
  swing2Motion=Free
  twistLimitDegrees=45
  twistMotion=Free
 }
]
