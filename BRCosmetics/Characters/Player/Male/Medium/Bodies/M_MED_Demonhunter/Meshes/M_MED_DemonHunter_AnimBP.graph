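# Secondary-motion AnimBP for the male Demon Hunter cosmetic body, copying the parent mesh's pose and animating jacket, collar and knee bones.
# The EventGraph maps arm, head and calf rotations, plus crouch and emote states, to alphas driving weighted constraints and additive corrective transforms.
# AnimDynamics simulates knee bones, followed by two RigidBody passes and an emote-only Control Rig.
# The hoodie chain is disconnected from output; leg-swing values are unused.
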AnimBlueprint /BRCosmetics/Characters/Player/Male/Medium/Bodies/M_MED_Demonhunter/Meshes/M_MED_DemonHunter_AnimBP
target_skeleton=&/Demon_Hunter_Skeleton
graphs [
 AnimGraph {nodes [
   n1 = "Constraint - dyn_hoodieFront_Base (PP)"!(ComponentPose, Alpha=1, ConstraintWeights_0=0.5, ConstraintWeights_1:float=0.5) -> (Pose) {
    alphaBoolBlend {blendOption=Linear}
    boneToModify {boneName=dyn_hoodieFront_Base}
    constraintSetup [{targetBone {boneName=spine_05}}, {targetBone {boneName=neck_01}}]
    constraintWeights [1,1]
   }
   n2 = "Constraint - dyn_hoodiePull_01_Base_r (PPPPPPPP)"!(
    ComponentPose=@n1
    Alpha=1
    ConstraintWeights_0=0.1181
    ConstraintWeights_1:float=0.004
    ConstraintWeights_2:float=0.007
    ConstraintWeights_3:float=0.0514
    ConstraintWeights_4:float=0.0269
    ConstraintWeights_5:float=0.018
    ConstraintWeights_6:float=0.0205
    ConstraintWeights_7:float=0.7543
   ) -> (Pose) {
    alphaBoolBlend {blendOption=Linear}
    boneToModify {boneName=dyn_hoodiePull_01_Base_r}
    constraintSetup [
     {targetBone {boneName=head}}
     {targetBone {boneName=spine_03}}
     {targetBone {boneName=spine_04}}
     {targetBone {boneName=spine_05}}
     {targetBone {boneName=clavicle_r}}
     {targetBone {boneName=neck_01}}
     {targetBone {boneName=neck_02}}
     {targetBone {boneName=dyn_hoodieFront}}
    ]
    constraintWeights [1,1,1,1,1,1,1,1]
   }
   n3 = "Constraint - dyn_hoodiePull_01_Base_l (PPPPPPPP)"!(
    ComponentPose=@n2
    Alpha=1
    ConstraintWeights_0=0.1181
    ConstraintWeights_1:float=0.004
    ConstraintWeights_2:float=0.007
    ConstraintWeights_3:float=0.0514
    ConstraintWeights_4:float=0.0269
    ConstraintWeights_5:float=0.018
    ConstraintWeights_6:float=0.0205
    ConstraintWeights_7:float=0.7543
   ) -> (Pose) {
    alphaBoolBlend {blendOption=Linear}
    boneToModify {boneName=dyn_hoodiePull_01_Base_l}
    constraintSetup [
     {targetBone {boneName=head}}
     {targetBone {boneName=spine_03}}
     {targetBone {boneName=spine_04}}
     {targetBone {boneName=spine_05}}
     {targetBone {boneName=clavicle_l}}
     {targetBone {boneName=neck_01}}
     {targetBone {boneName=neck_02}}
     {targetBone {boneName=dyn_hoodieFront}}
    ]
    constraintWeights [1,1,1,1,1,1,1,1]
   }
   n4 = "Transform (Modify) Bone - Bone: dyn_hoodieFront"(ComponentPose=@n3, Alpha=1, Translation="0,0,0", Rotation="0,0,0", Scale="1,1,1") {
    rotationMode=BMM_Additive
    rotationSpace=BCS_BoneSpace
    translationMode=BMM_Additive
    translationSpace=BCS_BoneSpace
    alphaBoolBlend {blendOption=Linear}
    boneToModify {boneName=dyn_hoodieFront}
   }
   "Transform (Modify) Bone - Bone: dyn_hoodieFront"(
    ComponentPose=@n4
    Alpha=Get headSwingDeltaBack() -> (headSwingDeltaBack:double)
    Translation="1.5,2,-0.000"
    Rotation="0.000,-10,0.000"
    Scale="1,1,1"
   ) {
    rotationMode=BMM_Additive
    rotationSpace=BCS_BoneSpace
    translationMode=BMM_Additive
    translationSpace=BCS_BoneSpace
    alphaBoolBlend {blendOption=Linear}
    boneToModify {boneName=dyn_hoodieFront}
    rotation {yaw=-10}
    translation=<1.5 2>
   }
   n5 = Get calfSwingDelta_r() -> (calfSwingDelta_r:double)
   n6 = "Constraint - dyn_jacketFront_Base_r (PPPPPP)"!(
    ComponentPose=Local To Component(LocalPose=Copy Pose From Mesh() {bUseAttachedParent=true})
    Alpha=1
    ConstraintWeights_0=0.2017
    ConstraintWeights_1:float=0.2923
    ConstraintWeights_2:float=0.3153
    ConstraintWeights_3:float=0.0732
    ConstraintWeights_4:float=0.0967
    ConstraintWeights_5:float=0.1127
   ) -> (Pose) {
    alphaBoolBlend {blendOption=Linear}
    boneToModify {boneName=dyn_jacketFront_Base_r}
    constraintSetup [
     {targetBone {boneName=spine_01}}
     {targetBone {boneName=spine_02}}
     {targetBone {boneName=spine_03}}
     {targetBone {boneName=spine_04}}
     {targetBone {boneName=clavicle_r}}
     {targetBone {boneName=pelvis}}
    ]
    constraintWeights [1,1,1,1,1,1]
   }
   n7 = "Constraint - dyn_jacketSide_Base_r (PPPPPP)"!(
    ComponentPose=@n6
    Alpha=1
    ConstraintWeights_0=0.2711
    ConstraintWeights_1:float=0.2987
    ConstraintWeights_2:float=0.4111
    ConstraintWeights_3:float=0.026
    ConstraintWeights_4:float=0.131
    ConstraintWeights_5:float=0.1673
   ) -> (Pose) {
    alphaBoolBlend {blendOption=Linear}
    boneToModify {boneName=dyn_jacketSide_Base_r}
    constraintSetup [
     {targetBone {boneName=spine_01}}
     {targetBone {boneName=spine_02}}
     {targetBone {boneName=spine_03}}
     {targetBone {boneName=spine_04}}
     {targetBone {boneName=clavicle_r}}
     {targetBone {boneName=pelvis}}
    ]
    constraintWeights [1,1,1,1,1,1]
   }
   n8 = "Constraint - dyn_jacketSideBack_Base_r (PPPPPP)"!(
    ComponentPose=@n7
    Alpha=1
    ConstraintWeights_0=0.2633
    ConstraintWeights_1:float=0.2204
    ConstraintWeights_2:float=0.547
    ConstraintWeights_3:float=0.0759
    ConstraintWeights_4:float=0.1775
    ConstraintWeights_5:float=0.1662
   ) -> (Pose) {
    alphaBoolBlend {blendOption=Linear}
    boneToModify {boneName=dyn_jacketSideBack_Base_r}
    constraintSetup [
     {targetBone {boneName=spine_01}}
     {targetBone {boneName=spine_02}}
     {targetBone {boneName=spine_03}}
     {targetBone {boneName=spine_04}}
     {targetBone {boneName=clavicle_r}}
     {targetBone {boneName=pelvis}}
    ]
    constraintWeights [1,1,1,1,1,1]
   }
   n9 = "Constraint - dyn_jacketFront_Base_l (PPPPPP)"!(
    ComponentPose=@n8
    Alpha=1
    ConstraintWeights_0=0.2017
    ConstraintWeights_1:float=0.2923
    ConstraintWeights_2:float=0.3153
    ConstraintWeights_3:float=0.0732
    ConstraintWeights_4:float=0.0967
    ConstraintWeights_5:float=0.1127
   ) -> (Pose) {
    alphaBoolBlend {blendOption=Linear}
    boneToModify {boneName=dyn_jacketFront_Base_l}
    constraintSetup [
     {targetBone {boneName=spine_01}}
     {targetBone {boneName=spine_02}}
     {targetBone {boneName=spine_03}}
     {targetBone {boneName=spine_04}}
     {targetBone {boneName=clavicle_l}}
     {targetBone {boneName=pelvis}}
    ]
    constraintWeights [1,1,1,1,1,1]
   }
   n10 = "Constraint - dyn_jacketSide_Base_l (PPPPPP)"!(
    ComponentPose=@n9
    Alpha=1
    ConstraintWeights_0=0.2711
    ConstraintWeights_1:float=0.2987
    ConstraintWeights_2:float=0.4111
    ConstraintWeights_3:float=0.026
    ConstraintWeights_4:float=0.131
    ConstraintWeights_5:float=0.1673
   ) -> (Pose) {
    alphaBoolBlend {blendOption=Linear}
    boneToModify {boneName=dyn_jacketSide_Base_l}
    constraintSetup [
     {targetBone {boneName=spine_01}}
     {targetBone {boneName=spine_02}}
     {targetBone {boneName=spine_03}}
     {targetBone {boneName=spine_04}}
     {targetBone {boneName=clavicle_l}}
     {targetBone {boneName=pelvis}}
    ]
    constraintWeights [1,1,1,1,1,1]
   }
   n11 = "Constraint - dyn_jacketSideBack_Base_l (PPPPPP)"!(
    ComponentPose=@n10
    Alpha=1
    ConstraintWeights_0=0.2633
    ConstraintWeights_1:float=0.2204
    ConstraintWeights_2:float=0.547
    ConstraintWeights_3:float=0.0759
    ConstraintWeights_4:float=0.1775
    ConstraintWeights_5:float=0.1662
   ) -> (Pose) {
    alphaBoolBlend {blendOption=Linear}
    boneToModify {boneName=dyn_jacketSideBack_Base_l}
    constraintSetup [
     {targetBone {boneName=spine_01}}
     {targetBone {boneName=spine_02}}
     {targetBone {boneName=spine_03}}
     {targetBone {boneName=spine_04}}
     {targetBone {boneName=clavicle_l}}
     {targetBone {boneName=pelvis}}
    ]
    constraintWeights [1,1,1,1,1,1]
   }
   n12 = "Constraint - dyn_jacketBack_Base (PPPPPPP)"!(
    ComponentPose=@n11
    Alpha=1
    ConstraintWeights_0=0.1432
    ConstraintWeights_1:float=0.0836
    ConstraintWeights_2:float=0.4324
    ConstraintWeights_3:float=0.1144
    ConstraintWeights_4:float=0.0492
    ConstraintWeights_5:float=0.1059
    ConstraintWeights_6:float=0.0713
   ) -> (Pose) {
    alphaBoolBlend {blendOption=Linear}
    boneToModify {boneName=dyn_jacketBack_Base}
    constraintSetup [
     {targetBone {boneName=spine_01}}
     {targetBone {boneName=spine_02}}
     {targetBone {boneName=spine_03}}
     {targetBone {boneName=spine_04}}
     {targetBone {boneName=clavicle_r}}
     {targetBone {boneName=pelvis}}
     {targetBone {boneName=clavicle_l}}
    ]
    constraintWeights [1,1,1,1,1,1,1]
   }
   n13 = "Constraint - thigh_split_l (P)"(ComponentPose=@n12, Alpha=Get Pouchalpha() -> (Pouchalpha:double), ConstraintWeights_0=1) {
    alphaBoolBlend {blendOption=Linear}
    boneToModify {boneName=thigh_split_l}
    constraintSetup [{targetBone {boneName=thigh_l}}]
    constraintWeights [1]
   }
   n14 = "Constraint - thigh_split_l (T)"(ComponentPose=@n13, Alpha=Get IsEmoting() -> (IsEmoting:double), ConstraintWeights_0=1) {
    alphaBoolBlend {blendOption=Linear}
    boneToModify {boneName=thigh_split_l}
    constraintSetup [{transformType=Translation, targetBone {boneName=thigh_l}}]
    constraintWeights [1]
   }
   n15 = Get calfSwingDelta_l() -> (calfSwingDelta_l:double)
   n16 = "Transform (Modify) Bone - Bone: dyn_kneeBack_Base_l"(ComponentPose=@n14, Alpha=@n15, Translation="15,3,0", Rotation="0,0,0", Scale="1,1,1") {
    rotationMode=BMM_Additive
    rotationSpace=BCS_BoneSpace
    translationMode=BMM_Additive
    translationSpace=BCS_BoneSpace
    alphaBoolBlend {blendOption=Linear}
    boneToModify {boneName=dyn_kneeBack_Base_l}
    translation=<15 3>
   }
   n17 = "Transform (Modify) Bone - Bone: dyn_kneeOut_Base_l"(ComponentPose=@n16, Alpha=@n15, Translation="5,2,0", Rotation="0,0,0", Scale="1,1,1") {
    rotationMode=BMM_Additive
    rotationSpace=BCS_BoneSpace
    translationMode=BMM_Additive
    translationSpace=BCS_BoneSpace
    alphaBoolBlend {blendOption=Linear}
    boneToModify {boneName=dyn_kneeOut_Base_l}
    translation=<5 2>
   }
   n18 = "Transform (Modify) Bone - Bone: dyn_kneeIn_Base_l"(ComponentPose=@n17, Alpha=@n15, Translation="4,2,0", Rotation="0,0,0", Scale="1,1,1") {
    rotationMode=BMM_Additive
    rotationSpace=BCS_BoneSpace
    translationMode=BMM_Additive
    translationSpace=BCS_BoneSpace
    alphaBoolBlend {blendOption=Linear}
    boneToModify {boneName=dyn_kneeIn_Base_l}
    translation=<4 2>
   }
   n19 = "Transform (Modify) Bone - Bone: dyn_kneeFront_Base_l"(ComponentPose=@n18, Alpha=@n15, Translation="2,2,0", Rotation="0.000,-10,0.000", Scale="1,1,1") {
    rotationMode=BMM_Additive
    rotationSpace=BCS_BoneSpace
    translationMode=BMM_Additive
    translationSpace=BCS_BoneSpace
    alphaBoolBlend {blendOption=Linear}
    boneToModify {boneName=dyn_kneeFront_Base_l}
    rotation {yaw=-10}
    translation=<2 2>
   }
   n20 = "Transform (Modify) Bone - Bone: dyn_kneeBack_Base_r"(ComponentPose=@n19, Alpha=@n5, Translation="-22,0,-2.5", Rotation="0,0,0", Scale="1,1,1") {
    rotationMode=BMM_Additive
    rotationSpace=BCS_BoneSpace
    translationMode=BMM_Additive
    translationSpace=BCS_BoneSpace
    alphaBoolBlend {blendOption=Linear}
    boneToModify {boneName=dyn_kneeBack_Base_r}
    translation {x=-22, z=-2.5}
   }
   n21 = "Transform (Modify) Bone - Bone: dyn_kneeOut_Base_r"(ComponentPose=@n20, Alpha=@n5, Translation="-5,-2,0", Rotation="0,0,0", Scale="1,1,1") {
    rotationMode=BMM_Additive
    rotationSpace=BCS_BoneSpace
    translationMode=BMM_Additive
    translationSpace=BCS_BoneSpace
    alphaBoolBlend {blendOption=Linear}
    boneToModify {boneName=dyn_kneeOut_Base_r}
    translation=<-5 -2>
   }
   n22 = "Transform (Modify) Bone - Bone: dyn_kneeIn_Base_r"(ComponentPose=@n21, Alpha=@n5, Translation="-4,-2,0", Rotation="0,0,0", Scale="1,1,1") {
    rotationMode=BMM_Additive
    rotationSpace=BCS_BoneSpace
    translationMode=BMM_Additive
    translationSpace=BCS_BoneSpace
    alphaBoolBlend {blendOption=Linear}
    boneToModify {boneName=dyn_kneeIn_Base_r}
    translation=<-4 -2>
   }
   n23 = "Transform (Modify) Bone - Bone: dyn_kneeFront_Base_r"(ComponentPose=@n22, Alpha=@n5, Translation="-2,-2,0", Rotation="0.000,-10,0.000", Scale="1,1,1") {
    rotationMode=BMM_Additive
    rotationSpace=BCS_BoneSpace
    translationMode=BMM_Additive
    translationSpace=BCS_BoneSpace
    alphaBoolBlend {blendOption=Linear}
    boneToModify {boneName=dyn_kneeFront_Base_r}
    rotation {yaw=-10}
    translation=<-2 -2>
   }
   n24 = Anim Dynamics - Bone: dyn_kneeOut_r(ComponentPose=@n23, Alpha=1, ExternalForce="0,0,0") {
    angularDampingOverride=1
    angularSpringConstant=20
    bAngularSpring=true
    bGravityOverrideInSimSpace=true
    bOverrideAngularDamping=true
    gravityScale=1.1
    simulationSpace=RootRelative
    alphaBoolBlend {blendOption=Linear}
    boundBone {boneName=dyn_kneeOut_r}
    componentAppliedLinearAccClamp=<2000 2000 2000>
    componentLinearAccScale=<0.4 0.4 0.4>
    componentLinearVelScale=<0.2 0.2 0.2>
    physicsBodyDefinitions [{
      sphereCollisionRadius=0
      boundBone {boneName=dyn_kneeOut_r}
      boxExtents=<15 4 20>
      constraintSetup {angularLimitsMax=<10 10 10>, angularLimitsMin=<-10 -10 -10>, angularTarget=<1 0 0>}
      localJointOffset=<5 0 0>
     }]
   }
   n25 = Anim Dynamics - Bone: dyn_kneeFront_r(ComponentPose=@n24, Alpha=1, ExternalForce="0,0,0") {
    angularDampingOverride=1
    angularSpringConstant=20
    bAngularSpring=true
    bGravityOverrideInSimSpace=true
    bOverrideAngularDamping=true
    gravityScale=1.1
    simulationSpace=RootRelative
    alphaBoolBlend {blendOption=Linear}
    boundBone {boneName=dyn_kneeFront_r}
    componentAppliedLinearAccClamp=<2000 2000 2000>
    componentLinearAccScale=<0.4 0.4 0.4>
    componentLinearVelScale=<0.2 0.2 0.2>
    physicsBodyDefinitions [{
      sphereCollisionRadius=0
      boundBone {boneName=dyn_kneeFront_r}
      boxExtents=<15 4 20>
      constraintSetup {angularLimitsMax=<10 10 10>, angularLimitsMin=<-10 -10 -10>, angularTarget=<1 0 0>}
      localJointOffset=<5 0 0>
     }]
   }
   n26 = Anim Dynamics - Bone: dyn_kneeIn_r(ComponentPose=@n25, Alpha=1, ExternalForce="0,0,0") {
    angularDampingOverride=1
    angularSpringConstant=20
    bAngularSpring=true
    bGravityOverrideInSimSpace=true
    bOverrideAngularDamping=true
    gravityScale=1.1
    simulationSpace=RootRelative
    alphaBoolBlend {blendOption=Linear}
    boundBone {boneName=dyn_kneeIn_r}
    componentAppliedLinearAccClamp=<2000 2000 2000>
    componentLinearAccScale=<0.4 0.4 0.4>
    componentLinearVelScale=<0.2 0.2 0.2>
    physicsBodyDefinitions [{
      sphereCollisionRadius=0
      boundBone {boneName=dyn_kneeIn_r}
      boxExtents=<15 4 20>
      constraintSetup {angularLimitsMax=<10 10 10>, angularLimitsMin=<-10 -10 -10>, angularTarget=<1 0 0>}
      localJointOffset=<5 0 0>
     }]
   }
   n27 = Anim Dynamics - Bone: dyn_kneeOut_l(ComponentPose=@n26, Alpha=1, ExternalForce="0,0,0") {
    angularDampingOverride=1
    angularSpringConstant=20
    bAngularSpring=true
    bGravityOverrideInSimSpace=true
    bOverrideAngularDamping=true
    gravityScale=1.1
    simulationSpace=RootRelative
    alphaBoolBlend {blendOption=Linear}
    boundBone {boneName=dyn_kneeOut_l}
    componentAppliedLinearAccClamp=<2000 2000 2000>
    componentLinearAccScale=<0.4 0.4 0.4>
    componentLinearVelScale=<0.2 0.2 0.2>
    physicsBodyDefinitions [{
      sphereCollisionRadius=0
      boundBone {boneName=dyn_kneeOut_l}
      boxExtents=<15 4 20>
      constraintSetup {angularLimitsMax=<10 10 10>, angularLimitsMin=<-10 -10 -10>, angularTarget=<1 0 0>}
      localJointOffset=<-5 0 0>
     }]
   }
   n28 = Anim Dynamics - Bone: dyn_kneeFront_l(ComponentPose=@n27, Alpha=1, ExternalForce="0,0,0") {
    angularDampingOverride=1
    angularSpringConstant=20
    bAngularSpring=true
    bGravityOverrideInSimSpace=true
    bOverrideAngularDamping=true
    gravityScale=1.1
    simulationSpace=RootRelative
    alphaBoolBlend {blendOption=Linear}
    boundBone {boneName=dyn_kneeFront_l}
    componentAppliedLinearAccClamp=<2000 2000 2000>
    componentLinearAccScale=<0.4 0.4 0.4>
    componentLinearVelScale=<0.2 0.2 0.2>
    physicsBodyDefinitions [{
      sphereCollisionRadius=0
      boundBone {boneName=dyn_kneeFront_l}
      boxExtents=<15 4 20>
      constraintSetup {angularLimitsMax=<10 10 10>, angularLimitsMin=<-10 -10 -10>, angularTarget=<1 0 0>}
      localJointOffset=<-5 0 0>
     }]
   }
   n29 = Anim Dynamics - Bone: dyn_kneeIn_l(ComponentPose=@n28, Alpha=1, ExternalForce="0,0,0") {
    angularDampingOverride=1
    angularSpringConstant=20
    bAngularSpring=true
    bGravityOverrideInSimSpace=true
    bOverrideAngularDamping=true
    gravityScale=1.1
    simulationSpace=RootRelative
    alphaBoolBlend {blendOption=Linear}
    boundBone {boneName=dyn_kneeIn_l}
    componentAppliedLinearAccClamp=<2000 2000 2000>
    componentLinearAccScale=<0.4 0.4 0.4>
    componentLinearVelScale=<0.2 0.2 0.2>
    physicsBodyDefinitions [{
      sphereCollisionRadius=0
      boundBone {boneName=dyn_kneeIn_l}
      boxExtents=<15 4 20>
      constraintSetup {angularLimitsMax=<10 10 10>, angularLimitsMin=<-10 -10 -10>, angularTarget=<1 0 0>}
      localJointOffset=<-5 0 0>
     }]
   }
   n30 = Get armSwingDelta_l() -> (armSwingDelta_l:double)
   n31 = "Constraint - dyn_collarFrontBase_l (P)"(ComponentPose=@n29, Alpha=@n30, ConstraintWeights_0=1) {
    alphaBoolBlend {blendOption=Linear}
    alphaScaleBiasClamp {bClampResult=true, clampMax=0.4}
    boneToModify {boneName=dyn_collarFrontBase_l}
    constraintSetup [{targetBone {boneName=clavicle_l}}]
    constraintWeights [1]
   }
   n32 = "Constraint - dyn_collarMidBase_l (P)"(ComponentPose=@n31, Alpha=@n30, ConstraintWeights_0=1) {
    alphaBoolBlend {blendOption=Linear}
    alphaScaleBiasClamp {bClampResult=true, clampMax=0.4}
    boneToModify {boneName=dyn_collarMidBase_l}
    constraintSetup [{targetBone {boneName=clavicle_l}}]
    constraintWeights [1]
   }
   n33 = Get armSwingDelta_r() -> (armSwingDelta_r:double)
   n34 = "Constraint - dyn_collarFrontBase_r (P)"(ComponentPose=@n32, Alpha=@n33, ConstraintWeights_0=1) {
    alphaBoolBlend {blendOption=Linear}
    alphaScaleBiasClamp {bClampResult=true, clampMax=0.4, clampMin=0.25}
    boneToModify {boneName=dyn_collarFrontBase_r}
    constraintSetup [{targetBone {boneName=clavicle_r}}]
    constraintWeights [1]
   }
   n35 = "Constraint - dyn_collarMidBase_r (P)"(ComponentPose=@n34, Alpha=@n33, ConstraintWeights_0=1) {
    alphaBoolBlend {blendOption=Linear}
    alphaScaleBiasClamp {bClampResult=true, clampMax=0.4, clampMin=0.25}
    boneToModify {boneName=dyn_collarMidBase_r}
    constraintSetup [{targetBone {boneName=clavicle_r}}]
    constraintWeights [1]
   }
   n36 = Get headSwingDeltaBack()
   n37 = "Transform (Modify) Bone - Bone: dyn_collarBack_Base"(ComponentPose=@n35, Alpha=@n36, Translation="0.5,3.75,-0.000", Rotation="0,18,0", Scale="1,1,1") {
    rotationMode=BMM_Additive
    rotationSpace=BCS_BoneSpace
    translationMode=BMM_Additive
    translationSpace=BCS_BoneSpace
    alphaBoolBlend {blendOption=Linear}
    alphaScaleBiasClamp {bMapRange=true, inRange {max=0.8, min=0.15}, outRange {min=0.1}}
    boneToModify {boneName=dyn_collarBack_Base}
    rotation {yaw=18}
    translation=<0.5 3.75>
   }
   n38 = "Transform (Modify) Bone - Bone: dyn_collarSide_Base_r"(ComponentPose=@n37, Alpha=@n36, Translation="0,0,0", Rotation="0,5,0", Scale="1,1,1") {
    rotationMode=BMM_Additive
    rotationSpace=BCS_BoneSpace
    translationMode=BMM_Additive
    translationSpace=BCS_BoneSpace
    alphaBoolBlend {blendOption=Linear}
    alphaScaleBiasClamp {bMapRange=true, inRange {max=0.8, min=0.15}, outRange {min=0.1}}
    boneToModify {boneName=dyn_collarSide_Base_r}
    rotation {yaw=5}
   }
   n39 = "Transform (Modify) Bone - Bone: dyn_collarMidBaseDriven_r"(ComponentPose=@n38, Alpha=@n36, Translation="0,0,0", Rotation="0,5,0", Scale="1,1,1") {
    rotationMode=BMM_Additive
    rotationSpace=BCS_BoneSpace
    translationMode=BMM_Additive
    translationSpace=BCS_BoneSpace
    alphaBoolBlend {blendOption=Linear}
    alphaScaleBiasClamp {bMapRange=true, inRange {max=0.8, min=0.15}, outRange {min=0.1}}
    boneToModify {boneName=dyn_collarMidBaseDriven_r}
    rotation {yaw=5}
   }
   n40 = "Transform (Modify) Bone - Bone: dyn_collarFrontBaseDriven_r"(ComponentPose=@n39, Alpha=@n33, Translation="-2.596,-1.066,-0.107", Rotation="5,-25,5", Scale="1,1,1") {
    rotationMode=BMM_Additive
    rotationSpace=BCS_BoneSpace
    translationMode=BMM_Additive
    translationSpace=BCS_BoneSpace
    alphaBoolBlend {blendOption=Linear}
    boneToModify {boneName=dyn_collarFrontBaseDriven_r}
    rotation=<rot 5 -25 5>
    translation=<-2.596 -1.066 -0.107>
   }
   n41 = Get armSwingDelta_l()
   n42 = "Transform (Modify) Bone - Bone: dyn_collarFrontBaseDriven_l"(ComponentPose=@n40, Alpha=@n41, Translation="0,0,0", Rotation="0,0,0", Scale="1,1,1") {
    rotationMode=BMM_Additive
    rotationSpace=BCS_BoneSpace
    translationMode=BMM_Additive
    translationSpace=BCS_BoneSpace
    alphaBoolBlend {blendOption=Linear}
    boneToModify {boneName=dyn_collarFrontBaseDriven_l}
   }
   n43 = "Transform (Modify) Bone - Bone: dyn_collarFrontBaseDriven_r"(ComponentPose=@n42, Alpha=@n33, Translation="1,1,0", Rotation="0,0,0", Scale="1,1,1") {
    translationMode=BMM_Additive
    translationSpace=BCS_BoneSpace
    alphaBoolBlend {blendOption=Linear}
    boneToModify {boneName=dyn_collarFrontBaseDriven_r}
    translation=<1 1>
   }
   n44 = "Transform (Modify) Bone - Bone: dyn_collarMidBaseDriven_r"(ComponentPose=@n43, Alpha=@n33, Translation="-1.983,-1.782,0.402", Rotation="-3.5,15,5", Scale="1,1,1") {
    rotationMode=BMM_Additive
    rotationSpace=BCS_BoneSpace
    translationMode=BMM_Additive
    translationSpace=BCS_BoneSpace
    alphaBoolBlend {blendOption=Linear}
    boneToModify {boneName=dyn_collarMidBaseDriven_r}
    rotation=<rot -3.5 15 5>
    translation=<-1.983 -1.782 0.402>
   }
   n45 = "Transform (Modify) Bone - Bone: dyn_collarMidBaseDriven_l"(ComponentPose=@n44, Alpha=@n41, Translation="0,0,0", Rotation="0,0,0", Scale="1,1,1") {
    rotationMode=BMM_Additive
    rotationSpace=BCS_BoneSpace
    translationMode=BMM_Additive
    translationSpace=BCS_BoneSpace
    alphaBoolBlend {blendOption=Linear}
    boneToModify {boneName=dyn_collarMidBaseDriven_l}
   }
   n46 = "Transform (Modify) Bone - Bone: dyn_collarMidBaseDriven_l"(ComponentPose=@n45, Alpha=@n41, Translation="-1,-1,0", Rotation="0,0,0", Scale="1,1,1") {
    translationMode=BMM_Additive
    translationSpace=BCS_BoneSpace
    alphaBoolBlend {blendOption=Linear}
    boneToModify {boneName=dyn_collarMidBaseDriven_l}
    translation=<-1 -1>
   }
   n47 = "Transform (Modify) Bone - Bone: dyn_collarSide_Base_r"(
    ComponentPose=@n46
    Alpha=Get headPitchAim() -> (headPitchAim:double)
    Translation="-0.8,1,0.5"
    Rotation="0.000,28.014,16.174"
    Scale="1,1,1"
   ) {
    rotationMode=BMM_Additive
    rotationSpace=BCS_BoneSpace
    translationMode=BMM_Additive
    translationSpace=BCS_BoneSpace
    alphaBoolBlend {blendOption=Linear}
    boneToModify {boneName=dyn_collarSide_Base_r}
    rotation {roll=16.1743, yaw=28.0138}
    translation=<-0.8 1 0.5>
   }
   n48 = "Transform (Modify) Bone - Bone: dyn_collarSide_Base_l"(ComponentPose=@n47, Alpha=Get armSwingDelta_l(), Translation="0,0,0", Rotation="0.000,-30,0.000", Scale="1,1,1") {
    rotationMode=BMM_Additive
    rotationSpace=BCS_BoneSpace
    translationMode=BMM_Additive
    translationSpace=BCS_BoneSpace
    alphaBoolBlend {blendOption=Linear}
    boneToModify {boneName=dyn_collarSide_Base_l}
    rotation {yaw=-30}
   }
   n49 = "Constraint - thigh_split_l (P)"(ComponentPose=@n48, Alpha=0.5, ConstraintWeights_0=1) {
    alphaBoolBlend {blendOption=Linear}
    boneToModify {boneName=thigh_split_l}
    constraintSetup [{targetBone {boneName=thigh_l}}]
    constraintWeights [1]
   }
   n50 = Get crouchAlpha() -> (crouchAlpha:bool)
   n51 = "Transform (Modify) Bone - Bone: dyn_jacketBack_Base"!(ComponentPose=@n49, bAlphaBoolEnabled:bool=@n50, Translation="0.000,0.5,0.000", Rotation="0.000,5,0.000", Scale="1,1,1") -> (Pose) {
    alphaInputType=Bool
    rotationMode=BMM_Additive
    rotationSpace=BCS_BoneSpace
    translationMode=BMM_Additive
    translationSpace=BCS_BoneSpace
    alphaBoolBlend {blendInTime=0.68, blendOption=Linear, blendOutTime=0.68}
    boneToModify {boneName=dyn_jacketBack_Base}
    rotation {yaw=5}
    translation {y=0.5}
   }
   n52 = "Transform (Modify) Bone - Bone: dyn_jacketSideBack_Base_r"!(ComponentPose=@n51, bAlphaBoolEnabled:bool=@n50, Translation="0.000,0.5,0.000", Rotation="0.000,5,0.000", Scale="1,1,1") -> (Pose) {
    alphaInputType=Bool
    rotationMode=BMM_Additive
    rotationSpace=BCS_BoneSpace
    translationMode=BMM_Additive
    translationSpace=BCS_BoneSpace
    alphaBoolBlend {blendInTime=0.68, blendOption=Linear, blendOutTime=0.68}
    boneToModify {boneName=dyn_jacketSideBack_Base_r}
    rotation {yaw=5}
    translation {y=0.5}
   }
   n53 = "Transform (Modify) Bone - Bone: dyn_jacketSideBack_Base_l"!(ComponentPose=@n52, bAlphaBoolEnabled:bool=@n50, Translation="0.000,-0.5,0.000", Rotation="0.000,5,0.000", Scale="1,1,1") -> (Pose) {
    alphaInputType=Bool
    rotationMode=BMM_Additive
    rotationSpace=BCS_BoneSpace
    translationMode=BMM_Additive
    translationSpace=BCS_BoneSpace
    alphaBoolBlend {blendInTime=0.68, blendOption=Linear, blendOutTime=0.68}
    boneToModify {boneName=dyn_jacketSideBack_Base_l}
    rotation {yaw=5}
    translation {y=-0.5}
   }
   n54 = RigidBody(ComponentPose=@n53, Alpha=Get armSwingINV_r() -> (armSwingINV_r:double)) {
    bClampLinearTranslationLimitToRefPose=true
    bOverrideWorldGravity=true
    overridePhysicsAsset=&/M_MED_DemonHunter_CollarPhysics
    simulationSpace=BaseBoneSpace
    alphaBoolBlend {blendOption=Linear}
    baseBoneRef {boneName=spine_05}
    componentAppliedLinearAccClamp=<4000 4000 4000>
    componentLinearAccScale=<0.2 0.2 0.2>
    componentLinearVelScale=<0.1 0.1 0.1>
    overrideWorldGravity {z=-980}
   }
   n55 = RigidBody(ComponentPose=@n54, Alpha=Get RbanAlpha() -> (RbanAlpha:double)) {
    bClampLinearTranslationLimitToRefPose=true
    bOverrideWorldGravity=true
    overridePhysicsAsset=&/M_MED_Demon_Hunter_Physics
    simulationSpace=BaseBoneSpace
    alphaBoolBlend {blendOption=Linear}
    baseBoneRef {boneName=spine_05}
    componentAppliedLinearAccClamp=<1000 1000 1000>
    componentLinearAccScale=<0.15 0.15 0.2>
    componentLinearVelScale=<0.15 0.15 0.2>
    overrideWorldGravity {z=-980}
   }
   n56 = Control Rig(Source=Component To Local(ComponentPose=@n55), Alpha=Get IsEmoting()) {
    alphaBoolBlend {blendOption=Linear}
    controlRigAssetReference {blueprintRigClass=&/M_MED_Demon_Hunter_CtrlRig.M_MED_Demon_Hunter_CtrlRig_C}
   }
   Output Pose(Result=@n56) {layerGroup=None}
   Get headSwingDeltaFront() -> (headSwingDeltaFront:double)
  ]}
 EventGraph {nodes [
   Event BlueprintUpdateAnimation()
   > n1 = Sequence!(execute) -> (then_0, then_1, then_2:Exec, then_3:Exec, then_4:Exec)
   n2 = GetOwnerSkeletalMeshComponent()
   n3 = BreakRotator(InRot=BreakTransform(InTransform=GetDeltaTransformFromRefPose(self=@n2, BoneName=upperarm_r, BaseName=clavicle_r)).Rotation)
   n4 = Set armSwingDelta_r(execute=@n1.then_1, armSwingDelta_r:double=MapRangeClamped(Value="Absolute (Float)"(A=@n3.Yaw), InRangeA=45, InRangeB=125)) -> (Output_Get:double)
   > Set armSwingINV_r(armSwingINV_r:double=float - float(B=@n4) {is=title K2Node_CallFunction}) -> (Output_Get:double)
   n5 = BreakRotator(InRot=BreakTransform(InTransform=GetDeltaTransformFromRefPose(self=@n2, BoneName=upperarm_l, BaseName=clavicle_l)).Rotation)
   > Set armSwingDelta_l(armSwingDelta_l:double=MapRangeClamped(Value="Absolute (Float)"(A=@n5.Yaw), InRangeA=5, InRangeB=77)) -> (Output_Get:double)
   n6 = BreakTransform(InTransform=GetDeltaTransformFromRefPose(self=@n2, BoneName=head, BaseName=spine_05))
   n7 = BreakRotator(InRot=@n6.Rotation)
   > Set headSwingDeltaBack(headSwingDeltaBack:double=MapRangeClamped(Value=@n7.Yaw, InRangeA=-3.5, InRangeB=-58)) -> (Output_Get:double)
   > Set headSwingDeltaFront(headSwingDeltaFront:double=MapRangeClamped(Value=@n7.Yaw, InRangeA=-3.5, InRangeB=27)) -> (Output_Get:double)
   n8 = BreakRotator(InRot=@n6.Rotation)
   > Set headTwist_r(headTwist_r:double=MapRangeClamped(Value=@n8, InRangeA=14, InRangeB=25)) -> (Output_Get:double)
   > Set headPitchAim(headPitchAim:double=MapRangeClamped(Value=@n8.Pitch, InRangeA=0, InRangeB=37)) -> (Output_Get:double)
   Set crouchAlpha(execute=@n1.then_3, crouchAlpha:bool=Get bIsCrouching()) -> (Output_Get:bool)
   > Set Pouchalpha(Pouchalpha:double=SelectFloat(A=0, B=1, bPickA=Get bRigidBodyEnabled())) -> (Output_Get:double)
   n9 = Get bIsPlayingEmote()
   > Set IsEmoting(IsEmoting:double=SelectFloat(A=1, B=0, bPickA=@n9)) -> (Output_Get:double)
   > Set RbanAlpha(RbanAlpha:double=SelectFloat(A=0.5, B=1, bPickA=@n9)) -> (Output_Get:double)
   Set legSwingDelta_r(
    execute=@n1.then_0
    legSwingDelta_r:double=MapRangeClamped(Value=GetTwistAndSwingAngleOfDeltaRotationFromRefPose(self=@n2, BoneName=thigh_r).OutSwingAngle, InRangeA=173, InRangeB=72)
   ) -> (Output_Get:double)
   > Set legSwingDelta_l(legSwingDelta_l:double=MapRangeClamped(Value=GetTwistAndSwingAngleOfDeltaRotationFromRefPose(self=@n2, BoneName=thigh_l).OutSwingAngle, InRangeA=24, InRangeB=101)) -> (Output_Get:double)
   Set calfSwingDelta_r(
    execute=@n1.then_2
    calfSwingDelta_r:double=MapRangeClamped(Value=GetTwistAndSwingAngleOfDeltaRotationFromRefPose(self=@n2, BoneName=calf_r).OutSwingAngle, InRangeA=20, InRangeB=120)
   ) -> (Output_Get:double)
   > Set calfSwingDelta_l(calfSwingDelta_l:double=MapRangeClamped(Value=GetTwistAndSwingAngleOfDeltaRotationFromRefPose(self=@n2, BoneName=calf_l).OutSwingAngle, InRangeA=20, InRangeB=120)) -> (Output_Get:double)
   TryGetPawnOwner()
   Reroute Node(InputPin:Wildcard) -> (OutputPin:Wildcard)
  ]}
]
variables [
 armSwingDelta_l {type=number}
 armSwingDelta_r {type=number}
 armSwingINV_r {type=number}
 calfSwingDelta_l {type=number}
 calfSwingDelta_r {type=number}
 crouchAlpha {type=boolean}
 headPitchAim {type=number}
 headSwingDeltaBack {type=number}
 headSwingDeltaFront {type=number}
 headTwist_r {type=number}
 IsEmoting {type=number}
 legSwingDelta_l {type=number}
 legSwingDelta_r {type=number}
 Pouchalpha {type=number}
 RbanAlpha {type=number}
]
